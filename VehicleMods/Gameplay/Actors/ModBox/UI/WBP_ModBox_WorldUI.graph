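# World UI label widget for vehicle mod boxes, shown only while the owning player is in a vehicle that can take a mod.
# HUD context vehicle-change events and construction refresh visibility, testing the vehicle with CanApplyMod against a mod tag unless a skip flag is set.
# While shown, squared-distance thresholds pick between two inherited label states.
# SetIsSpawner rescales the root overlay for the spawner timer variant.
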WidgetBlueprint /VehicleMods/Gameplay/Actors/ModBox/UI/WBP_ModBox_WorldUI
graphs [
 EventGraph {nodes [
   Event Destruct()
   > Parent: Destruct()
   n1 = Get FortHUDContext!() -> (ReturnValue:Fort HUDContext Object Reference) {is=title K2Node_GetBlueprintContextEx}
   > Is Valid(InputObject=@n1)
   n2 = Create Event()
   > Unbind Event from On Player Vehicle State Changed(self:Fort HUDContext Object Reference=@n1, Delegate=@n2) {is=title K2Node_RemoveDelegate}
   Event Event Display Error Message!() -> (OutputDelegate, then, ShouldDisplayError:bool) {is=title K2Node_Event}
   > Parent: Event Display Error Message!(execute, ShouldDisplayError:bool=false) -> (then) {is=title K2Node_CallParentFunction}
   n3 = Event OnDistanceUpdate!() -> (OutputDelegate, then, PrevDistanceThreshold:int=0, NextDistanceThreshold:int=0) {is=title K2Node_Event}
   > SetStateFromThreshold!(execute, self, Threshold:int=@n3.NextDistanceThreshold) -> (then)
   Event Construct()
   > Parent: Construct()
   > Bind Event to On Player Vehicle State Changed(self:Fort HUDContext Object Reference=@n1, Delegate=@n2) {is=title K2Node_AddDelegate}
   > Set SqrDistanceThresholds(SqrDistanceThresholds:"Array of Float (single-precision)s"=Get Box_ModdableView() -> (Box_ModdableView:"Array of Float (single-precision)s")) -> (Output_Get:"Array of Float (single-precision)s")
   > RefreshVehicleStateForPawn!(execute, self, Player:Pawn Object Reference=GetOwningPlayerPawn!(self:User Widget Object Reference) -> (ReturnValue:Pawn Object Reference)) -> (then)
  ]}
 OnPlayerVehicleStateChanged {nodes [
   n1 = OnPlayerVehicleStateChanged!() -> (then, PlayerPawn:Fort Player Pawn Object Reference, NewVehicle:Actor Object Reference, OldVehicle:Actor Object Reference) {is=title K2Node_FunctionEntry}
   > RefreshVehicleStateForPawn(Player=@n1)
  ]}
 RefreshVehicleStateForPawn {nodes [
   n1 = RefreshVehicleStateForPawn!() -> (then, Player:Pawn Object Reference) {is=title K2Node_FunctionEntry}
   n2 = GetOwningPlayerPawn()
   > Branch(Condition="Equal (Object)"(A=@n1, B=@n2))
   > n3 = Is in Vehicle!(execute:Exec, Player:Pawn Object Reference=@n2) -> (true:Exec, false:Exec, Vehicle Occupied:Fort Vehicle Interface Interface) {is=title K2Node_MacroInstance}
   > n4 = Branch(Condition=Get bSkipCanApplyModCheck() -> (bSkipCanApplyModCheck:bool))
   UpdateVehicleStateUI!(execute=[@n5.then, @n4.then], self, IsInModdableVehicle:bool=true) -> (then)
   n5 = Branch(
    execute=@n4.else
    Condition=CanApplyMod!(self:Fort Vehicle Interface Interface=@n3, InTag:"Gameplay Tag Structure (by ref)"=Get TestModTag() -> (TestModTag:Gameplay Tag Structure)) -> (ReturnValue:bool)
   )
   UpdateVehicleStateUI(execute=[@n5.else, @n3.false], IsInModdableVehicle=false)
  ]}
 SetIsSpawner {nodes [
   n1 = SetIsSpawner!() -> (then, IsSpawnerTimerUI:bool=~) {is=title K2Node_FunctionEntry}
   > n2 = Set Is Spawner Timer UI(Is Spawner Timer UI:bool=@n1) -> (Output_Get:bool)
   > SetRenderScale!(
    execute
    self:Widget Object Reference=Get Overlay_Root() -> (Overlay_Root:Overlay Object Reference)
    Scale:Vector 2D Structure=Select(Option 0:Vector 2D Structure=(X=0.85,Y=0.85), Option 1:Vector 2D Structure=(X=1.1,Y=1.1), Index=@n2) -> (ReturnValue:Vector 2D Structure)
   ) -> (then)
  ]}
 SetStateFromThreshold {nodes [
   n1 = SetStateFromThreshold!() -> (then, Threshold:int) {is=title K2Node_FunctionEntry}
   > Branch(Condition=Get IsPlayerInModdableVehicle() -> (IsPlayerInModdableVehicle:bool))
   > SetState!(
    execute
    self:WBP World UI Label Object Reference
    State:ENUM_WorldUI_States Enum=Select(Option 0:ENUM_WorldUI_States Enum=NewEnumerator2, Option 1:ENUM_WorldUI_States Enum=NewEnumerator0, Index="Equal (Integer)"(A=@n1)) -> (ReturnValue:ENUM_WorldUI_States Enum)
   ) -> (then)
  ]}
 UpdateVehicleStateUI {nodes [
   n1 = UpdateVehicleStateUI!() -> (then, IsInModdableVehicle:bool=~) {is=title K2Node_FunctionEntry}
   > n2 = Set IsPlayerInModdableVehicle(IsPlayerInModdableVehicle:bool=@n1) -> (Output_Get:bool)
   > n3 = Branch(Condition=@n2)
   > SetVisibility(InVisibility=HitTestInvisible)
   > SetStateFromThreshold(Threshold=Get CurrentDistanceThreshold() -> (CurrentDistanceThreshold:int=0)) {is=title K2Node_CallFunction}
   SetVisibility(execute=@n3.else, InVisibility=Hidden)
  ]}
]
variables [
 Box_ModdableView {type="array<number>", default [5600000]}
 bSkipCanApplyModCheck {type=boolean}
 Is Spawner Timer UI {type=boolean}
 IsPlayerInModdableVehicle {type=boolean}
 TestModTag {type=GameplayTag, default {tagName=Vehicle.Mod}}
]
widgets {
 info {inheritedWidgetCount=0, parentClass=/Game/Athena/HUD/WorldUI/WBP_WorldUI_Label.WBP_WorldUI_Label_C, rootWidgetClass=null, widgetCount=0}
 widgets []
}
